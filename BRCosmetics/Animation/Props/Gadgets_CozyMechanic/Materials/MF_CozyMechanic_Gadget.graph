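# Material function for the Cozy Mechanic gadget that adds a burning pipe glow with sparks to the incoming material attributes' emissive.
# A soft sphere mask in tilted local X-Z space, distorted by animated gradient noise, is boosted by a panning star texture for sparks.
# Named reroutes supply a pipe texture mask, the clamped MaterialCurveA animation input and an inverted Fresnel for the glow's masking and two-color blend.
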MaterialFunction /BRCosmetics/Animation/Props/Gadgets_CozyMechanic/Materials/MF_CozyMechanic_Gadget
graph {nodes [
  Constant() {r=1}
  FunctionInput() {inputName=TechArt Mask}
  n1 = GetMaterialAttributes(None=FunctionInput() {inputType=FunctionInput_MaterialAttributes}) {attributeGetTypes [#1]}
  n2 = NamedRerouteDeclaration(None=@n1.EmissiveColor) {name=Emissive_In, nodeColor=<rgba 1 0.3281 0 1>}
  n3 = LocalPosition()
  n4 = ScalarParameter() {defaultValue=0.01, parameterName=PipeBurningNoiseITiling}
  n5 = Time()
  n6 = ScalarParameter() {defaultValue=0.2, parameterName=PipeBurningNoiseIPanningSpeed}
  n7 = Noise!(World Position=Add(A=Multiply(A=@n3.XYZ, B=@n4) {constB=0.05}, B=Multiply(A=@n5, B=@n6)), FilterWidth) {noiseFunction=NOISEFUNCTION_GradientTex3D}
  n8 = AppendVector(A=Constant(), B=Multiply(A=ComponentMask(None=@n7) {r=true}, B=ScalarParameter() {defaultValue=10, parameterName=PipeBurningNoiseIntensity}))
  n9 = LocalPosition()
  n10 = RotateAboutAxis(
   NormalizedRotationAxis=Constant3Vector() {constant=<rgba 1 0 0 0>}
   RotationAngle=ScalarParameter() {defaultValue=-0.04}
   PivotPoint=Constant3Vector()
   Position=LocalPosition().XYZ
  )
  n11 = MaterialFunctionCall!(Float4=AppendVector(A=Add(A=@n9.XYZ, B=@n10), B=Constant())) {materialFunction=/Game/Global/MaterialFunctions/MF_BreakComponents}
  n12 = AppendVector(A=Multiply(A=@n11.R) {constB=0.6}, B=@n11.B)
  n13 = ScalarParameter() {parameterName=PipeBurningOrigin_X}
  n14 = ScalarParameter() {defaultValue=22, parameterName=PipeBurningOrigin_Z}
  n15 = Power(
   Base=SphereMask(A=Add(A=@n8, B=@n12), B=AppendVector(A=@n13, B=@n14), Radius=ScalarParameter() {defaultValue=40, parameterName=PipeBurningRadius}) {hardnessPercent=0}
   Exp=ScalarParameter() {defaultValue=1.5, parameterName=PipeBurningMaskPow}
  )
  n16 = TextureCoordinate()
  n17 = Constant2Vector() {g=1, r=2}
  n18 = Constant()
  n19 = ScalarParameter() {defaultValue=-0.05, parameterName=SparkPanningSpeed}
  n20 = TextureSample(UVs=Panner(Coordinate=Multiply(A=@n16, B=@n17), Speed=AppendVector(A=@n18, B=@n19)) {speedY=0.1}) {samplerType=SAMPLERTYPE_LinearColor, texture=/Game/Packages/Fortress_Sky/SkyDome/Master/T_SkyDome_Stars_Sharp}
  n21 = ScalarParameter() {defaultValue=5, parameterName=SparkEmissiveMult}
  n22 = Add(A=@n15, B=Multiply(A=@n15, B=Multiply(A=@n20.R, B=@n21)))
  n23 = LinearInterpolate(
   A=VectorParameter() {parameterName=PipeBurningColor_Min, defaultValue=<rgba 0.5 0 0 1>}.RGB
   B=VectorParameter() {parameterName=PipeBurningColor_Max, defaultValue=<rgba 5 0.5 0 1>}.RGB
   Alpha=NamedRerouteUsage()
  )
  n24 = Multiply(A=NamedRerouteUsage(), B=Multiply(A=NamedRerouteUsage(), B=Multiply(A=@n22, B=@n23)))
  FunctionOutput(None=SetMaterialAttributes!(MaterialAttributes=@n1.MaterialAttributes, EmissiveColor=Add(A=@n2, B=@n24)) {attributeSetTypes [#1]})
  NamedRerouteDeclaration(None=TextureSample() {texture=/BRCosmetics/Animation/Props/Gadgets_CozyMechanic/Textures/Temp_CozyMechanic_PipeMap}.R) {name=PipeMask, nodeColor=<rgba 0 1 0.6484 1>}
  NamedRerouteDeclaration(None=Clamp(None=ScalarParameter() {parameterName=MaterialCurveA}) {minDefault=0.2}) {name=AnimInput, nodeColor=<rgba 0 0.8125 1 1>}
  NamedRerouteDeclaration(None=MaterialFunctionCall(
    Normal Vector=VertexNormalWS()
    Invert Fresnel=StaticBool() {value=true}
    Power=ScalarParameter() {defaultValue=1, parameterName=PipeFresnelPower_Min}
   ) {materialFunction=&Fresnel_Function}.Result) {name=FresnelMask, nodeColor=<rgba 1 0.8672 0 1>}
  Saturate(None=Multiply(A=MaterialFunctionCall() {materialFunction=&DebugTimeSine}.Result) {constB=5})
 ]}
